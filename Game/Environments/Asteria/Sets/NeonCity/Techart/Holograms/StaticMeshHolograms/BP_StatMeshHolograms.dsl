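; Neon City hologram Blueprint that displays a configurable mesh with projector beam, god rays, optional audio and local post-process.
; Initialize applies the mesh, and Glow Color tints the beam, god rays and post-process.
; Visibility and audio follow either a proximity trigger, debounced by retriggerable delays, or, when proximity is off, optionally evening/night phases.
; A separate collision volume ramps post-process blend weight via timeline for the locally viewed pawn only.
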(event EventBeginPlay
  (bind (_scenefringeintensity _bloomintensity _autoexposurebias) (Utilities|Struct|BreakPostProcessSettings (Variables|/Script/Engine.PostProcessComponent|GetSettings (Variables|Default|GetPostProcess))))
  (bind _postprocess (Variables|Default|GetPostProcess))
  (Utilities|FlowControl|Sequence
    (:then_0
      (CallFunction|Initialize)
      (if (Variables|Default|GetProximityTrigger)
        (Rendering|SetVisibility (Variables|Default|GetHologramMesh) false false)
        (CallFunction|UpdateAudio false)
        (else
          (Collision|SetCollisionEnabled (Variables|Default|GetTrigger%20Box) "NoCollision")
          ;[1 1 1] ONly SHow Hologram at Night
          (if (Variables|Default|GetNight%20Time%20Only)
            (bind _timeofdaymanager (Class|FortAsyncActionTimeOfDayManagerReady|GetTimeOfDayManagerReady)
              (:OnReady
                (TimeOfDay|RegisterforDayPhaseChangedEvent (TimeOfDay|GetContextualTimeOfDayManager) (delegate Custom|Time%20of%20Day%20Update) false)))))))
    (:then_1
      ;[1 1 1] Audio Play Loop
      (if (Variables|Default|GetPlay%20Audio)
        (bind _audio (Variables|Default|GetAudio)
          (:then
            (Audio|Components|Audio|Play _audio 0.0))
          (:else))
        (else
          (Components|DestroyComponent (Variables|Default|GetAudio)))))
    (:then_2
      (bind _output_get (Variables|/Script/Engine.PostProcessComponent|SetSettings (Utilities|Struct|MakePostProcessSettings (Variables|Default|GetGlow%20Color) _scenefringeintensity _bloomintensity _autoexposurebias) (Variables|Default|GetPostProcess)))
      (Variables|/Script/Engine.PostProcessComponent|SetSettings _output_get _postprocess))))

(event Custom|Time%20of%20Day%20Update
  (params (CurrentDayPhase :type EFortDayPhase%20Enum :default "Morning") (PreviousDayPhase :type EFortDayPhase%20Enum :default "Morning") (bAtCreation :type Boolean))
  ;[1 1 1] ONly SHow Hologram at Night
  (CallFunction|NightTimeOnline CurrentDayPhase))

(event Custom|OnAudioFinished_Event)

(event EventTick
  (params (DeltaSeconds :type Float%20%28single-precision%29)))

(event OnComponentBeginOverlap(TriggerBox)
  (params (OverlappedComponent :type Primitive%20Component%20Object%20Reference) (OtherActor :type Actor%20Object%20Reference) (OtherComp :type Primitive%20Component%20Object%20Reference) (OtherBodyIndex :type Integer) (bFromSweep :type Boolean) (SweepResult :type Hit%20Result%20Structure%20%28by%20ref%29))
  ;[1 1 1] Trigger system
  (if (Variables|Default|GetProximityTrigger)
    (Rendering|SetVisibility (Variables|Default|GetHologramMesh) true true)
    (CallFunction|UpdateAudio true)
    (Variables|Default|SetHologramActive true)
    ;;[1 1 1] Prevents rapidly going in and out
    (Variables|Default|SetFlickerprevent true)
    (Utilities|FlowControl|RetriggerableDelay 0.5)
    (Variables|Default|SetFlickerprevent false)))

(event OnComponentEndOverlap(TriggerBox)
  (params (OverlappedComponent :type Primitive%20Component%20Object%20Reference) (OtherActor :type Actor%20Object%20Reference) (OtherComp :type Primitive%20Component%20Object%20Reference) (OtherBodyIndex :type Integer))
  ;[1 1 1] Trigger system
  (if (Variables|Default|GetProximityTrigger)
    (if (Variables|Default|GetHologramActive)
      (if (Variables|Default|GetFlickerprevent)
        (Utilities|FlowControl|RetriggerableDelay 2.0))
      (Rendering|SetVisibility (Variables|Default|GetHologramMesh) false true)
      (CallFunction|UpdateAudio false)
      (Variables|Default|SetHologramActive false))))

(event OnComponentBeginOverlap(PostProcessCollision)
  (params (OverlappedComponent :type Primitive%20Component%20Object%20Reference) (OtherActor :type Actor%20Object%20Reference) (OtherComp :type Primitive%20Component%20Object%20Reference) (OtherBodyIndex :type Integer) (bFromSweep :type Boolean) (SweepResult :type Hit%20Result%20Structure%20%28by%20ref%29))
  ;[1 1 1] Post Process Ramp
  (Networking|CanExecuteCosmeticEvents
    (:True
      (bind _asfort_player_pawn (Utilities|Casting|CastToFortPlayerPawn OtherActor))
      (if (Pawn|IsLocallyViewed _asfort_player_pawn)
        (bind (_direction _float) (Timeline|Ramp%20Post PlayFromStart)
          (:Update
            (Variables|/Script/Engine.PostProcessComponent|SetBlendWeight _float (Variables|Default|GetPostProcess))))))))

(event OnComponentEndOverlap(PostProcessCollision)
  (params (OverlappedComponent :type Primitive%20Component%20Object%20Reference) (OtherActor :type Actor%20Object%20Reference) (OtherComp :type Primitive%20Component%20Object%20Reference) (OtherBodyIndex :type Integer))
  ;[1 1 1] Post Process Ramp
  (Networking|CanExecuteCosmeticEvents
    (:True
      (bind _asfort_player_pawn (Utilities|Casting|CastToFortPlayerPawn OtherActor))
      (if (Pawn|IsLocallyViewed _asfort_player_pawn)
        (Timeline|Ramp%20Post ReverseFromEnd)))))

(fn Initialize
  (bind _god_rays (Variables|Default|GetGod%20Rays))
  (Utilities|FlowControl|Sequence
    (:then_0
      (Components|StaticMesh|SetStaticMesh (Variables|Default|GetHologramMesh) (Variables|Default|GetMesh)))
    (:then_1
      (bind _returnvalue (Rendering|Material|CreateDynamicMaterialInstance (Variables|Default|GetProjectionBase) 0 (Rendering|Material|GetMaterial (Variables|Default|GetProjectionBase) 0) "None" :!class /Script/Engine.PrimitiveComponent))
      (Rendering|Material|SetVectorParameterValue _returnvalue "Beam Color" (Variables|Default|GetGlow%20Color) :!class /Script/Engine.MaterialInstanceDynamic)
      (Rendering|Material|SetVectorParameterValueonMaterials _god_rays "Color A" "0.700000,0.700000,0.800000")
      (Rendering|Material|SetVectorParameterValueonMaterials _god_rays "Color B" (* (Math|Conversions|ToVector(LinearColor) (Variables|Default|GetGlow%20Color)) "2.000000,2.000000,2.000000")))))

(fn NightTimeOnline
  (params (NewParam :type EFortDayPhase%20Enum :default "Morning"))
  (bind _hologrammesh (Variables|Default|GetHologramMesh))
  (if (or (Utilities|Enum|Equal(Enum) NewParam "Night") (Utilities|Enum|Equal(Enum) NewParam "Evening"))
    (Rendering|SetVisibility _hologrammesh true true)
    (CallFunction|UpdateAudio true)
    (else
      (Rendering|SetVisibility _hologrammesh false true)
      (CallFunction|UpdateAudio false))))

(fn UpdateAudio
  (params (Play? :type Boolean))
  (bind _audio (Variables|Default|GetAudio))
  (bind _audio_2 (Variables|Default|GetAudio))
  (Variables|Default|GetAudio
    (:then
      (if Play?
        (Audio|Components|Audio|FadeIn _audio 0.25 1.0 0.0 "Linear" :!class /Script/Engine.AudioComponent)
        (Audio|PlaySoundatLocation (Variables|Sound|GetOnSound) (Transformation|GetWorldLocation _audio))
        (else
          (Audio|Components|Audio|FadeOut _audio_2 0.25 0.0 "Linear" :!class /Script/Engine.AudioComponent)
          (Audio|PlaySoundatLocation (Variables|Sound|GetOffSound) (Transformation|GetWorldLocation _audio_2)))))
    (:else)))

(fn ConstructionScript
  (bind (_colorcontrast _colorcontrastshadows _scenefringeintensity _bloomintensity _autoexposurebias) (Utilities|Struct|BreakPostProcessSettings (Variables|/Script/Engine.PostProcessComponent|GetSettings (Variables|Default|GetPostProcess))))
  (CallFunction|Initialize)
  (Variables|/Script/Engine.PostProcessComponent|SetSettings (Utilities|Struct|MakePostProcessSettings _colorcontrast _colorcontrastshadows (Variables|Default|GetGlow%20Color) _scenefringeintensity _bloomintensity _autoexposurebias) (Variables|Default|GetPostProcess)))
