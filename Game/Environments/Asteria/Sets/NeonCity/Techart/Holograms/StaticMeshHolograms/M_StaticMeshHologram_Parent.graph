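# Parent material for Neon City mesh holograms: opaque, two-sided Lit, faking transparency with skylight samples and fresnel.
# Color combines curve- and time-driven chromatic aberration, pixelation glitches and a height-based tint, feeding base color and emissive with scanlines and sunlight passthrough.
# An animation texture, optionally an atlas, drives vertex offset and normals.
# Dithered, noise-modulated pixel depth offset adds breakup; quality switches simplify costly effects for low-end and mobile.
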Material /Game/Environments/Asteria/Sets/NeonCity/Techart/Holograms/StaticMeshHolograms/M_StaticMeshHologram_Parent
graph {
 properties {
  bUseMaterialAttributes=true
  bUsedWithClothing=true
  bUsedWithInstancedStaticMeshes=true
  bUsedWithMeshDeformer=true
  bUsedWithSkeletalMesh=true
  ditherOpacityMask=true
  pixelDepthOffsetMode=PDOM_Legacy
  twoSided=true
  usedShadingModels=Default Lit
 }
 nodes [
  Add()
  Constant()
  MaterialFunctionCall!(Blend Range, Start Offset) {materialFunction=/Engine/Functions/Engine_MaterialFunctions03/Particles/Distance_Blend}
  n1 = Constant() {r=1}
  n2 = MaterialFunctionCall(Value_Morning=@n1, Value_Day=@n1, Value_Evening=@n1, Value_Night=ScalarParameter() {defaultValue=5, parameterName=Night Boost}) {materialFunction=&Scalar_Time_of_Day}
  Multiply(A=@n2.Result) {constB=0.5}
  NamedRerouteDeclaration(None=MaterialFunctionCall!(
    Default=MaterialFunctionCall!(Scanline Amount) {materialFunction=/Game/Environments/Asteria/Sets/NeonCity/Techart/Holograms/NewHologramMaterial/MFScanlines}.Result
    Low/Mobile=Constant() {r=1}
   ) {materialFunction=/Game/Effects/Fort_Effects/Materials/_Functions/MF_LowMobile_PathingSwitch}.Result) {name=Scanlines, nodeColor=<rgba 1 0 0.1641 1>}
  n3 = NamedRerouteUsage()
  n4 = Multiply(A=Time(), B=ScalarParameter() {defaultValue=0.02, parameterName=Glitch Curve Speed})
  n5 = CurveAtlasRowParameter(CurveTime=Frac(None=Add(A=@n3, B=@n4))) {
   atlas=/Game/Environments/Asteria/Sets/NeonCity/Techart/Holograms/NewHologramMaterial/HologramCurves
   curve=/Game/Environments/Asteria/Sets/NeonCity/Techart/Holograms/NewHologramMaterial/HologramCurve
   parameterName=Glitch
  }
  NamedRerouteDeclaration(None=@n5.G) {name=Glitch Green, nodeColor=<rgba 1 0.5625 0 1>}
  n6 = ActorPositionWS()
  NamedRerouteDeclaration(None=Divide(A=MaterialFunctionCall(Float3=@n6) {materialFunction=&BreakOutFloat3Components}.G) {constB=20.111}) {name=Position Random, nodeColor=<rgba 0 0.0156 1 1>}
  n7 = TextureObjectParameter() {parameterName=AnimationTexture, samplerType=SAMPLERTYPE_LinearColor, texture=/Game/Global/Textures/Default/Blanks/T_BlackVectorDisplacement}
  n8 = MaterialFunctionCall(Float2=TextureProperty(Tex=@n7)) {materialFunction=&BreakOutFloat2Components}
  NamedRerouteDeclaration(None=Divide(B=@n8.G)) {name=texelHeight, nodeColor=<rgba 0 1 0.2266 1>}
  n9 = MaterialFunctionCall(Float3=OneMinus(None=CameraVectorWS())) {materialFunction=&BreakOutFloat3Components}
  n10 = Add(A=@n9.B, B=ScalarParameter() {defaultValue=-1, parameterName=Offset SeeThrough})
  n11 = Multiply(
   A=SkyLightEnvMapSample(
    Direction=MaterialFunctionCall(X=@n9.R, Y=@n9.R, Z=@n10) {materialFunction=&MakeFloat3}.Result
    Roughness=ScalarParameter() {defaultValue=0.1, parameterName=HologramRoughness}
   )
   B=ScalarParameter() {defaultValue=1, parameterName=CubemapBleed}
  )
  n12 = Constant4Vector() {constant=<rgba 0.2031 0.1723 0.1315 1>}
  n13 = Constant4Vector() {constant=<rgba 0.0373 0.0718 0.1302 1>}
  NamedRerouteDeclaration(None=QualitySwitch(
    Default=Multiply(A=@n11, B=@n2.Result)
    Low=MaterialFunctionCall(Color_Morning=@n12.RGB, Color_Day=@n12.RGB, Color_Evening=@n12.RGB, Color_Night=@n13.RGB) {materialFunction=&Color_Time_of_Day}.Result
   )) {name=Cubemap, nodeColor=<rgba 0.8984 0 1 1>}
  n14 = Divide(B=NamedRerouteDeclaration(None=@n8.R) {name=textureWidth, nodeColor=<rgba 0.0078 0 1 1>})
  NamedRerouteDeclaration(None=Multiply(A=NamedRerouteDeclaration(None=@n14) {name=texelWidth, nodeColor=<rgba 0 1 0.2266 1>}) {constB=0.5}) {name=halfTexelWidth, nodeColor=<rgba 0.0312 0 1 1>}
  n15 = NamedRerouteUsage()
  n16 = NamedRerouteUsage()
  n17 = Multiply(A=Constant() {r=1}, B=@n16)
  n18 = Multiply(A=@n16) {constB=0.5}
  n19 = TextureSample(
   UVs=MaterialFunctionCall(X=NamedRerouteUsage(), Y=Add(A=@n17, B=@n18)) {materialFunction=&MakeFloat2}.Result
   Tex=NamedRerouteUsage()
  ) {texture=/Engine/EngineResources/DefaultTexture}
  NamedRerouteDeclaration(None=Add(A=@n15, B=@n19.R)) {name=uStart, nodeColor=<rgba 1 0 0.0469 1>}
  NamedRerouteDeclaration(None=Subtract(A=@n19.B, B=@n15)) {name=uWidth, nodeColor=<rgba 0 1 0.6016 1>}
  NamedRerouteDeclaration(None=@n7) {name=AnimationTexture, nodeColor=<rgba 0.6172 0 1 1>}
  NamedRerouteDeclaration(None=@n19.G) {name=vStart, nodeColor=<rgba 0.6875 0 1 1>}
  NamedRerouteDeclaration(None=@n19.A) {name=vHeight, nodeColor=<rgba 0 1 0.2266 1>}
  n20 = NamedRerouteUsage()
  n21 = NamedRerouteUsage()
  n22 = Multiply(A=Time(), B=ScalarParameter() {defaultValue=0.75, parameterName=Anim Speed}) {constB=0.5}
  n23 = ScalarParameter() {parameterName=TimeOffset}
  n24 = Multiply(A=Add(A=Add(A=@n22, B=@n23), B=PerInstanceRandom()), B=Divide(B=NamedRerouteUsage()) {constA=30})
  n25 = Multiply(A=Divide(B=Add(A=@n21) {constB=0.0001}), B=@n24)
  n26 = NamedRerouteUsage()
  n27 = DynamicParameter() {defaultValue=<rgba 0 0 0 0>}
  n28 = NamedRerouteUsage()
  n29 = Multiply(A=Divide(B=Add(A=@n26) {constB=0.0001}), B=Multiply(A=@n27.Param2, B=Divide(B=@n28) {constA=30}))
  n30 = LinearInterpolate(A=Multiply(A=@n21, B=Frac(None=@n25)), B=Multiply(A=@n26, B=Frac(None=@n29)), Alpha=@n27.Param1)
  n31 = NamedRerouteUsage()
  n32 = Multiply(A=Frac(None=@n24), B=Subtract(B=@n31))
  n33 = NamedRerouteUsage()
  n34 = Multiply(A=@n33) {constB=1.5}
  n35 = Multiply(A=@n33) {constB=0.5}
  n36 = MaterialFunctionCall!(
   u=StaticSwitchParameter(True=Add(A=@n30, B=@n20), False=Add(A=@n32, B=@n31)) {parameterName=TextureIsAnimationAtlas}
   bind_u=Add(A=StaticSwitchParameter(True=@n34, False=@n35) {parameterName=TextureIsAnimationAtlas})
   vStart=StaticSwitchParameter(True=NamedRerouteUsage(), False=Constant()) {parameterName=TextureIsAnimationAtlas}
   vHeight=StaticSwitchParameter(True=NamedRerouteUsage(), False=Constant() {r=1}) {parameterName=TextureIsAnimationAtlas}
   texelHeight=NamedRerouteUsage()
   WorldNormal
  ) {materialFunction=/Game/ArtTools/characterWPO/MF_WPOTransform_subuvs}
  NamedRerouteDeclaration(None=@n36.WPO) {name=Animation WPO, nodeColor=<rgba 0 1 0.7656 1>}
  n37 = TextureSampleParameter2D() {parameterName=Normals, samplerType=SAMPLERTYPE_Normal, texture=&FlatNormalBlue}
  n38 = Constant3Vector() {constant=<rgba 0.5 0.5 1 0>}
  NamedRerouteDeclaration(None=MaterialFunctionCall(
    BaseNormal=@n36.Normal
    AdditionalNormal=MaterialFunctionCall(Default=@n37.RGB, Low/Mobile=@n38) {materialFunction=&MF_LowMobile_PathingSwitch}.Result
   ) {materialFunction=&BlendAngleCorrectedNormals}.Result) {name=Normals, nodeColor=<rgba 1 0 0.4922 1>}
  n39 = Multiply(A=Time(), B=ScalarParameter() {defaultValue=0.33, group=Glitching, parameterName=Glitch Speed})
  NamedRerouteDeclaration(None=Multiply(A=@n39) {constB=3}) {name=GlitchUV, nodeColor=<rgba 1 0.5859 0 1>}
  NamedRerouteDeclaration(None=@n5.B) {name=Glitch Blue, nodeColor=<rgba 1 0.5625 0 1>}
  n40 = Multiply(A=Add(A=NamedRerouteUsage(), B=Time()), B=ScalarParameter() {defaultValue=0.33, group=Glitching, parameterName=Glitch Speed})
  n41 = LinearInterpolate(B=ScalarParameter() {parameterName=ChromaticAb}, Alpha=Clamp(None=Sine(None=@n40) {period=20}) {minDefault=-1})
  n42 = Constant()
  n43 = TextureCoordinate()
  n44 = TextureObjectParameter() {parameterName=BaseColor, texture=/Engine/EngineResources/Black_Low.Black}
  n45 = TextureSample(
   UVs=Add(A=@n43, B=MaterialFunctionCall(R=@n41, G=@n42, B=@n42) {materialFunction=&AppendMany}.RG)
   Tex=@n44
  )
  n46 = Add(A=@n43, B=MaterialFunctionCall(R=@n42, G=@n41, B=@n42) {materialFunction=&AppendMany}.RG)
  n47 = Add(
   A=@n43
   B=MaterialFunctionCall(R=@n42, G=@n42, B=OneMinus(None=@n41)) {materialFunction=&AppendMany}.RG
  )
  n48 = QualitySwitch(
   Default=MaterialFunctionCall(R=@n45.R, G=TextureSample(UVs=@n46, Tex=@n44).G, B=TextureSample(UVs=@n47, Tex=@n44).B) {materialFunction=&AppendMany}.RGB
   Low=@n45.RGB
   Medium=@n45.RGB
  )
  n49 = ScalarParameter() {defaultValue=30, parameterName=Tile}
  n50 = Multiply(A=Frac(None=@n40)) {constB=10}
  n51 = LinearInterpolate(A=Multiply(A=@n49) {constB=0.8}, B=@n49, Alpha=Divide(A=Round(None=@n50)) {constB=10})
  n52 = Round(None=Multiply(A=@n43, B=@n51) {constB=5})
  n53 = Multiply(A=Frac(None=NamedRerouteUsage())) {constB=0.58}
  n54 = TextureSample(UVs=LinearInterpolate(A=@n43, B=Divide(A=@n52, B=@n51), Alpha=Round(None=@n53)), Tex=@n44)
  n55 = Round(None=NamedRerouteUsage())
  n56 = TransformPosition(None=WorldPosition().XYZ) {transformSourceType=TRANSFORMPOSSOURCE_World}
  n57 = MaterialFunctionCall() {materialFunction=&ObjectLocalBounds}
  n58 = LinearInterpolate(
   A=ScalarParameter() {defaultValue=0.8, parameterName=Hologram Tint Min}
   B=ScalarParameter() {defaultValue=0.4, parameterName=Hologram Tint Max}
   Alpha=MaterialFunctionCall(Float3=Divide(A=@n56, B=@n57.Local Bounds Size)) {materialFunction=&BreakOutFloat3Components}.B
  )
  NamedRerouteDeclaration(None=LinearInterpolate(
    A=LinearInterpolate(A=@n48, B=@n54.RGB, Alpha=@n55)
    B=VectorParameter() {parameterName=Overall Hologram Tint, defaultValue=<rgba 1 0.2604 0.2879 1>}.RGB
    Alpha=Saturate(None=@n58)
   )) {name=Color, nodeColor=<rgba 0.3125 0 1 1>}
  n59 = TextureSampleParameter2D() {parameterName=SRM, samplerType=SAMPLERTYPE_Masks, texture=/Game/Global/Textures/Default/Blanks/Temp_specular_Mask}
  n60 = MaterialFunctionCall(Default=@n59.B, Low/Mobile=ScalarParameter() {defaultValue=0.5, parameterName=MobileMetallic}) {materialFunction=&MF_LowMobile_PathingSwitch}
  n61 = MaterialFunctionCall(Default=@n59.G, Low/Mobile=ScalarParameter() {parameterName=MobileRoughness}) {materialFunction=&MF_LowMobile_PathingSwitch}
  n62 = StaticBoolParameter() {parameterName=SwizzleSRM}
  n63 = ScalarParameter() {defaultValue=8, parameterName=LumenBrightness}
  n64 = ScalarParameter() {defaultValue=4, parameterName=SeethroughSidesFres}
  n65 = NamedRerouteUsage()
  n66 = LinearInterpolate(
   B=ScalarParameter() {defaultValue=3, parameterName=fake Transparency - Max amount}
   Alpha=Saturate(None=Fresnel(ExponentIn=@n64, Normal=@n65) {baseReflectFraction=0})
  ) {constB=0}
  n67 = NamedRerouteUsage()
  n68 = Multiply(A=@n67, B=ScalarParameter() {defaultValue=1.5, parameterName=Color Outer})
  n69 = Multiply(A=@n67, B=ScalarParameter() {defaultValue=0.2, parameterName=Color inner})
  n70 = Saturate(None=Fresnel(Normal=NamedRerouteUsage()) {baseReflectFraction=0, exponent=3})
  n71 = ScalarParameter() {defaultValue=0.8, parameterName=MobileBrightness}
  n72 = Add(
   A=LinearInterpolate(A=@n54.RGB, B=NamedRerouteUsage(), Alpha=Saturate(None=@n66))
   B=QualitySwitch(Default=LinearInterpolate(A=@n68, B=@n69, Alpha=@n70), Low=Multiply(A=@n71, B=@n67))
  )
  n73 = NamedRerouteUsage()
  n74 = Saturate(None=CollectionParameter() {collection=&FortniteMaterialParameters, parameterName=SunLightColor})
  n75 = Multiply(
   A=ScalarParameter() {defaultValue=0.02, parameterName=Sunlight Passthrough}
   B=Multiply(A=ComponentMask(None=@n74) {b=true, g=true, r=true}, B=NamedRerouteUsage())
  )
  n76 = Clamp(None=MaterialFunctionCall() {materialFunction=/Game/Packages/Fortress_Sky/Fortnite_SunScreenSpaceFalloff}.Dot*-1)
  n77 = Add(A=Clamp(None=Multiply(A=@n72, B=@n73)), B=QualitySwitch(Default=Multiply(A=@n75, B=@n76), Low=Constant()))
  n78 = MaterialFunctionCall!(
   Emissive=Multiply(A=@n77, B=ScalarParameter() {defaultValue=0.6, parameterName=EmissiveBrightness})
   Alpha=ScalarParameter() {defaultValue=1, group=Cosmetic Emissive EyeAdaptation, parameterName=Cosmetic Emissive EyeAdaptation Alpha}
   Material
  ) {materialFunction=/Game/Packages/Fortress_SharedMaterials/Base_Material_Functions/BaseCharacterShaderFunctions/MF_CosmeticEmissiveEyeAdaptation}
  n79 = Multiply(
   A=Multiply(A=ScalarParameter() {defaultValue=150, parameterName=PDO_Distance}, B=ScalarParameter() {defaultValue=1, parameterName=PDO_DistanceMult})
   B=MaterialFunctionCall() {materialFunction=&DitherTemporalAA}.Result
  )
  n80 = MaterialFunctionCall(Float3=WorldPosition().XYZ) {materialFunction=&BreakOutFloat3Components}
  n81 = AppendVector(A=@n80.G, B=@n80.G)
  n82 = ScalarParameter() {defaultValue=50, group=Glitching, parameterName=GlitchScale}
  n83 = TextureSample(UVs=Add(A=@n39, B=Divide(A=@n81, B=@n82))) {
   samplerType=SAMPLERTYPE_Masks
   texture=/Game/Environments/Asteria/Sets/NeonCity/Techart/Holograms/F_MED_BlueBadass_NeonLines_-_GlitchNoise_v1
  }
  n84 = WorldPosition() {worldPositionShaderOffset=WPT_ExcludeAllShaderOffsets}
  n85 = CameraPositionWS()
  n86 = Divide(A=Add(A=Distance(A=@n84.XYZ, B=@n85), B=Constant() {r=-25000}), B=Constant() {r=50000})
  n87 = Constant() {r=4}
  n88 = Multiply(A=Multiply(A=Clamp(None=@n83.RGB) {minDefault=0.1}, B=Constant() {r=1500000}), B=Saturate(None=Power(Base=@n86, Exp=@n87)))
  n89 = Constant()
  SetMaterialAttributes!(
   MaterialAttributes
   Base Color=Multiply(A=@n77, B=ScalarParameter() {defaultValue=0.6, parameterName=Base Color Multiplier})
   Metallic=StaticSwitch(True=@n60.Result, False=@n61.Result, Value=@n62)
   Specular=MaterialFunctionCall(Default=@n59.R, Low/Mobile=ScalarParameter() {defaultValue=0.5, parameterName=MobileSpecular}) {materialFunction=&MF_LowMobile_PathingSwitch}.Result
   Roughness=StaticSwitch(True=@n61.Result, False=@n60.Result, Value=@n62)
   Anisotropy
   Emissive Color=RayTracingQualitySwitch(Normal=@n78.Emissive, RayTraced=Multiply(A=@n78.Emissive, B=@n63))
   Opacity
   Opacity Mask
   Normal=NamedRerouteUsage()
   Tangent
   World Position Offset=MaterialFunctionCall!(WPO=NamedRerouteUsage()) {materialFunction=/Game/Packages/Fortress_SharedMaterials/Base_Material_Functions/Fortnite_HideElementsWithVertexColor_2022}.Result
   Subsurface Color
   Pixel Depth Offset=QualitySwitch(Default=Add(A=@n79, B=@n88), Low=@n89, Medium=@n89)
   Custom Data 1
  ) {attributeSetTypes [
    #1
    #2
    #3
    #4
    #5
    #6
    #7
    #8
    #9
    #10
    #11
    #12
    #13
    #14
   ]}
  VertexInterpolator(VS=PreSkinnedPosition())
 ]
}
